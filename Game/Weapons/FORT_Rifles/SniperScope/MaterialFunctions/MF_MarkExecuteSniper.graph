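# Material function that draws the custom Mark and Execute sniper scope crosshair over the scene texture, passing it through when Enable MarkExecuteScope is off.
# It combines T_Crosshair channels, including a horizontally shifted blue-channel sample, with procedural center circle and atan2-based arc masks.
# Mark parameters appear to drive element movement and reddening.
# The combined masks blend crosshair color into the scene texture, using supplied Mask and Crosshair UV inputs.
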MaterialFunction /Game/Weapons/FORT_Rifles/SniperScope/MaterialFunctions/MF_MarkExecuteSniper
graph {nodes [
  ComponentMask(None=NamedRerouteUsage()) {g=true}
  DotProduct(A=Normalize(VectorInput=NamedRerouteUsage()), B=Constant2Vector() {g=0.5, r=0.5})
  n1 = FunctionInput() {bUsePreviewValueAsDefault=true, inputName=Scene Texture}
  FunctionOutput(None=StaticSwitchParameter(True=NamedRerouteUsage(), False=NamedRerouteDeclaration(None=@n1) {name=Scene Texture Input, nodeColor=<rgba 1 0.5859 0 1>}) {group=Custom Crosshair, parameterName=Enable MarkExecuteScope})
  n2 = NamedRerouteUsage()
  LinearInterpolate(
   A=LinearInterpolate(B=Constant3Vector() {constant=<rgba 1 0.0336 0.0172 1>}, Alpha=Ceil(None=@n2))
   B=Constant3Vector() {constant=<rgba 1 0.0336 0.0172 1>}
   Alpha=Ceil(None=NamedRerouteUsage())
  )
  n3 = ComponentMask(None=NamedRerouteUsage()) {g=true}
  n4 = Constant() {r=0.605}
  n5 = OneMinus(None=Saturate(None=Step(Y=@n3, X=@n4)))
  n6 = NamedRerouteDeclaration(None=Multiply(A=NamedRerouteDeclaration(None=@n5) {name=Blue Channel Mask, nodeColor=<rgba 1 0 0.75 1>}, B=NamedRerouteUsage())) {name=Arrow, nodeColor=<rgba 0 1 0.7422 1>}
  Multiply(A=@n6, B=NamedRerouteUsage())
  NamedRerouteDeclaration(None=FunctionInput() {bUsePreviewValueAsDefault=true, inputName=Mask, inputType=FunctionInput_Scalar, sortPriority=100}) {name=Mask, nodeColor=<rgba 0 1 0.9063 1>}
  n7 = TextureSampleParameter2D(UVs=Saturate(None=NamedRerouteUsage())) {parameterName=T_Crosshair, samplerType=SAMPLERTYPE_LinearColor, texture=/Game/Weapons/FORT_Rifles/Textures/Crosshair_Dot_M}
  NamedRerouteDeclaration(None=@n7.R) {name=Red Channel, nodeColor=<rgba 0.0078 0 1 1>}
  NamedRerouteDeclaration(None=@n7.G) {name=Green Channel, nodeColor=<rgba 0.0078 0 1 1>}
  NamedRerouteDeclaration(None=@n7.B) {name=Blue Channel, nodeColor=<rgba 0.0078 0 1 1>}
  n8 = NamedRerouteUsage()
  n9 = Saturate(None=@n8)
  n10 = Subtract(A=ComponentMask(None=@n8) {r=true}) {constB=0.5}
  n11 = NamedRerouteUsage()
  n12 = Multiply(
   A=MaterialFunctionCall(X=Sign(None=@n10), Y=Constant() {r=1}) {materialFunction=&MakeFloat2}.Result
   B=Multiply(A=OneMinus(None=@n11), B=ScalarParameter() {defaultValue=0.1, parameterName=Shift Distance})
  )
  NamedRerouteDeclaration(None=TextureSampleParameter2D(UVs=Add(A=@n9, B=@n12)) {parameterName=T_Crosshair, samplerType=SAMPLERTYPE_LinearColor, texture=&Crosshair_Dot_M}.B) {name=Blue Channel UV Shift, nodeColor=<rgba 0.0078 0 1 1>}
  n13 = NamedRerouteUsage()
  n14 = NamedRerouteUsage()
  n15 = NamedRerouteUsage()
  n16 = VectorParameter() {parameterName=Arc Center, defaultValue=<rgba -0.5 -0.5 0 1>}
  n17 = VectorParameter() {parameterName=Arc Center Mark, defaultValue=<rgba -0.5 -0.235 0 1>}
  n18 = NamedRerouteUsage()
  n19 = Add(A=NamedRerouteUsage(), B=ComponentMask(None=LinearInterpolate(A=@n16.RGB, B=@n17.RGB, Alpha=@n18)) {g=true, r=true}) {constB=0.5}
  n20 = ComponentMask(None=@n19) {g=true}
  n21 = Divide(
   A=Arctangent2!(Y=ComponentMask(None=@n19) {r=true}, X=Multiply(A=@n20) {constB=-1})
   B=MaterialFunctionCall() {materialFunction=&Pi}.Result
  )
  n22 = ScalarParameter() {defaultValue=0.025, parameterName=Arc Min}
  n23 = ScalarParameter() {defaultValue=0.06, parameterName=Arc Max}
  n24 = NamedRerouteUsage()
  n25 = Constant() {r=0.25}
  n26 = ComponentMask(None=NamedRerouteUsage()) {g=true}
  n27 = Constant() {r=1}
  n28 = Constant()
  n29 = Multiply(
   A=SphereMask(A=ConstantBiasScale(None=@n21), B=Constant() {r=0.5}, Radius=LinearInterpolate(A=@n22, B=@n23, Alpha=@n24) {constA=0.025, constB=0.06})
   B=Multiply(A=If(A=@n25, B=@n26, "A > B"=@n27, "A < B"=@n28), B=NamedRerouteUsage())
  )
  n30 = Multiply(A=SphereMask(A=NamedRerouteUsage(), B=Constant2Vector() {g=0.5, r=0.5}) {attenuationRadius=0.17}, B=NamedRerouteUsage())
  n31 = NamedRerouteUsage()
  n32 = InverseLinearInterpolate(Value=OneMinus(None=ComponentMask(None=@n31) {g=true})) {constA=0.4, constB=0.6}
  n33 = Add(
   A=Add(A=NamedRerouteDeclaration(None=@n29) {name=Top Arc, nodeColor=<rgba 0 1 0.3906 1>}, B=@n6)
   B=Multiply(A=NamedRerouteDeclaration(None=@n30) {name=Center Circle, nodeColor=<rgba 0 1 0.7422 1>}, B=Saturate(None=@n32))
  )
  n34 = NamedRerouteUsage()
  n35 = Max(
   A=SmoothStep(Value=Max(A=@n14, B=@n15)) {constMin=0.15}
   B=NamedRerouteDeclaration(None=Multiply(A=@n33, B=@n34)) {name=Red Elements, nodeColor=<rgba 0 1 0.0625 1>}
  )
  NamedRerouteDeclaration(None=LinearInterpolate(A=NamedRerouteUsage(), B=NamedRerouteUsage(), Alpha=Multiply(A=@n35, B=@n13))) {name=Scene Texture Output, nodeColor=<rgba 0.2656 0 1 1>}
  NamedRerouteDeclaration(None=FunctionInput() {bUsePreviewValueAsDefault=true, inputName=Crosshair UV, inputType=FunctionInput_Vector2, sortPriority=200}) {name=Crosshair UV, nodeColor=<rgba 0.0547 1 0 1>}
  n36 = NamedRerouteUsage()
  n37 = Constant3Vector() {constant=<rgba 1 0.2536 0.1043 0>}
  n38 = Constant3Vector() {constant=<rgba 1 0.0336 0.0172 1>}
  n39 = Ceil(None=NamedRerouteUsage())
  n40 = NamedRerouteUsage()
  NamedRerouteDeclaration(None=LinearInterpolate(A=LinearInterpolate(Alpha=@n36) {constA=1, constB=0}, B=LinearInterpolate(A=@n37, B=@n38, Alpha=@n36), Alpha=Multiply(A=@n39, B=@n40)) {constAlpha=0}) {name=Crosshair Color, nodeColor=<rgba 1 0.0703 0 1>}
  n41 = NamedRerouteUsage()
  n42 = NamedRerouteUsage()
  NamedRerouteDeclaration(None=Max(A=Max(A=@n41, B=@n42), B=NamedRerouteUsage())) {name=Full Mask, nodeColor=<rgba 0 0.25 1 1>}
  n43 = ScalarParameter() {parameterName=Use Mark}
  NamedRerouteDeclaration(None=@n43) {name="Use Mark&&Execute", nodeColor=<rgba 0.8281 1 0 1>}
  NamedRerouteDeclaration(None=Multiply(A=@n43, B=ScalarParameter() {defaultValue=1, parameterName=Marked})) {name=Marked, nodeColor=<rgba 0.8516 0 1 1>}
 ]}
